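# Conversation database for Mystica's main-questline dialogue in the Season 24 narrative quest Part 03, Q04.
# The entry tag leads to an auto-selected speech choice, gated on the quest being active and not completed, its first objective unfinished, and the NPC carrying the Mystica conversation tag.
# Two speech lines about The Unseen play in sequence, then a task completes that quest objective by backend name.
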ConversationDatabase /BattlePassS24/Conversations/Part03/Conversation_S24_P03_Q04
entry_points [{tag=Conversation.Entry.Season24.Narrative.P03.Q04.Mystica, destinations [#1]}]
nodes [
 ConversationEntryPointNode_0 {
  id=#1
  class=ConversationEntryPointNode
  outputs [#2]
  properties {entryTag {tagName=Conversation.Entry.Season24.Narrative.P03.Q04.Mystica}}
 }
 Speech_C_3 {
  id=#2
  class=Speech_C
  outputs [#3]
  properties {
   softTaskIcon=&T_UI_NPC_MainQuest
   generalConfig {defaultMessage="If The Unseen are working with The Army of the Last Reality, they are fools."}
  }
  sub_nodes [
   ConversationChoiceNode_1 {
    class=ConversationChoiceNode
    properties {
     defaultChoiceDisplayText=Talk to Mystica
     choiceTags {gameplayTags [{tagName=Conversation.Choice.AutoSelectIfValid}, {tagName=Conversation.Choice.Category.MainQuestline}]}
    }
   }
   FortConversationRequirement_ControllerRequirement_0 {
    class=FortConversationRequirement_ControllerRequirement
    properties {
     nodeName=Require: Has Completed Quest
     requirement {
      class=FortControllerRequirement_HasCompletedQuest
      properties {bInverseCondition=true, questDefinition=/BattlePassS24/Items/QuestItems/NarrativeQuests/Part03/quest_s24_narrative_p03_q04}
     }
    }
   }
   FortConversationRequirement_ControllerRequirement_1 {
    class=FortConversationRequirement_ControllerRequirement
    properties {
     nodeName=Require: Is Quest Active
     requirement {
      class=FortControllerRequirement_IsQuestActive
      properties {questDefinition=&quest_s24_narrative_p03_q04}
     }
    }
   }
   FortConversationRequirement_ControllerRequirement_2 {
    class=FortConversationRequirement_ControllerRequirement
    properties {
     nodeName=Require: Has Completed Quest Objective
     requirement {
      class=FortControllerRequirement_HasCompletedQuestObjective
      properties {
       bInverseCondition=true
       objectiveBackendName=quest_s24_narrative_p03_q04_obj0
       questDefinition=&quest_s24_narrative_p03_q04
      }
     }
    }
   }
   FortConversationRequirement_ControllerRequirement_3 {
    class=FortConversationRequirement_ControllerRequirement
    properties {
     nodeName=Require: NPC Has Conversation Tag
     participantId {tagName=Conversation.Participant.NPC}
     requirement {class=FortControllerRequirement_NPCHasConversationEntryTag, properties {queryTag {tagName=Conversation.Entry.Mystica}}}
    }
   }
  ]
 }
 Speech_C_28 {
  id=#3
  class=Speech_C
  outputs [#4]
  properties {generalConfig {defaultMessage=We must learn the truth before it is too late. Give us some time to work.}}
  sub_nodes [ConversationChoiceNode_0 {class=ConversationChoiceNode, properties {defaultChoiceDisplayText=Continue}}]
 }
 FortConversationTaskNode_CompleteQuestObjectiveByBackendName_1 {
  id=#4
  class=FortConversationTaskNode_CompleteQuestObjectiveByBackendName
  properties {
   objectiveBackendName=quest_s24_narrative_p03_q04_obj0
   questDefinition=&quest_s24_narrative_p03_q04
  }
 }
]
